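# Test unlit translucent deferred decal for Flaming Hoops that draws an animated orange, fiery ring-shaped glow.
# Sphere masks on LightVector components build the soft ring, broken up by panning cell, fire and streak noise into a Radial Mask.
# Named reroutes carry the mask, decal alpha and a two-color diffuse lerp.
# An emissive adjustment function takes a reroute scaled by an Emissive Multiplier.
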Material /FlamingHoopsGameplay/FX/M_Decal_FlamingHoops_Test
graph {
 properties {
  bAutomaticallySetUsageInEditor=false
  bCastRayTracedShadows=false
  bTangentSpaceNormal=false
  blendMode=BLEND_Translucent
  materialDomain=MD_DeferredDecal
  pixelDepthOffsetMode=PDOM_Legacy
  shadingModel=MSM_Unlit
  usedShadingModels=Unlit
 }
 nodes [
  Constant()
  MaterialFunctionCall(Emissive=Multiply(A=NamedRerouteUsage(), B=ScalarParameter() {defaultValue=1, parameterName=Emissive Multiplier})) {materialFunction=&MF_FX_EmissiveAdjustments}
  NamedRerouteDeclaration(None=LinearInterpolate(
    A=VectorParameter() {parameterName=Base Color, defaultValue=<rgba 1 0.0894 0 0.5>}.RGB
    B=VectorParameter() {parameterName=Base Color A, defaultValue=<rgba 5 1.183 0 1>}.RGB
    Alpha=NamedRerouteUsage()
   ) {constAlpha=1}) {name=Diffuse, nodeColor=<rgba 0 0.5078 1 1>}
  NamedRerouteDeclaration(None=NamedRerouteUsage()) {name=Opacity, nodeColor=<rgba 0 0.2969 1 1>}
  NamedRerouteDeclaration(None=DecalColor().A) {name=Decal Opacity, nodeColor=<rgba 1 0 0.4922 1>}
  n1 = LightVector()
  n2 = Constant() {r=0.5}
  n3 = LightVector()
  n4 = ComponentMask(None=@n3) {b=true, r=true}
  n5 = Constant() {r=0.5}
  n6 = OneMinus(None=SphereMask(A=@n4, B=@n5, Radius=@n5) {hardnessPercent=0})
  n7 = Constant() {r=0.5}
  n8 = ScalarParameter() {defaultValue=0.66, parameterName=Radius Hardness}
  n9 = Multiply(
   A=SphereMask(A=ComponentMask(None=@n1) {g=true}, B=@n2, Radius=@n2, Hardness=ScalarParameter() {defaultValue=0.66, parameterName=Radius Hardness})
   B=Multiply(A=Power(Base=@n6) {constExponent=3.5}, B=SphereMask(A=@n4, B=@n7, Radius=@n7, Hardness=@n8))
  )
  n10 = NamedRerouteUsage()
  n11 = ComponentMask(None=@n1) {g=true, r=true}
  n12 = Constant2Vector() {g=0.15, r=0.15}
  n13 = TextureSample(UVs=Panner(Coordinate=Multiply(A=@n11, B=@n12)) {speedX=-0.2, speedY=0.01}) {texture=/Game/Global/Textures/Default/Noise/T_CellNoise_01}
  n14 = Panner(Coordinate=Multiply(A=ComponentMask(None=@n1) {g=true, r=true}, B=Constant2Vector() {g=0.1, r=0.75})) {speedX=-0.05, speedY=-0.01}
  n15 = Multiply(
   A=LinearInterpolate(Alpha=OneMinus(None=@n13.G)) {constA=0.3, constB=0.7}
   B=LinearInterpolate(Alpha=TextureSample(UVs=@n14) {texture=/Game/Global/Textures/Default/Noise/T_FireNoise_01}.G) {constA=0.3, constB=0.8}
  )
  n16 = Multiply(A=ComponentMask(None=@n3) {g=true, r=true}, B=Constant2Vector() {g=1, r=8}) {constA=1}
  n17 = Power(Base=TextureSampleParameter2D(UVs=Panner(Coordinate=@n16) {speedX=-2.5}) {group=AlphaNoise, parameterName=AlphaNoise2, texture=&/T_FX_FlamingHoop_PannerNoise1}.G)
  NamedRerouteDeclaration(None=Multiply(A=Multiply(A=@n9, B=@n10), B=LinearInterpolate(A=@n15, Alpha=@n17))) {name=Radial Mask, nodeColor=<rgba 0 0.9766 1 1>}
  NamedRerouteUsage()
  Subtract(A=@n16, B=Constant2Vector())
 ]
}
